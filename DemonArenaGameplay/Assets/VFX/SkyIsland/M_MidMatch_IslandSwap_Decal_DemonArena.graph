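# Unlit translucent deferred decal for the Demon Arena mid-match island swap effect, drawing a soft radial glow, blue by default.
# Color blends dark to bright using a UV sphere mask times a cosine gradient over world Z relative to the actor, scaled to decal height.
# Opacity reuses the sphere mask with separate shaping, multiplied by caustic-texture erosion, which the default threshold leaves fully opaque.
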Material /DemonArenaGameplay/Assets/VFX/SkyIsland/M_MidMatch_IslandSwap_Decal_DemonArena
graph {
 properties {
  blendMode=BLEND_Translucent
  materialDomain=MD_DeferredDecal
  pixelDepthOffsetMode=PDOM_Legacy
  shadingModel=MSM_Unlit
  usedShadingModels=Unlit
 }
 nodes [
  n1 = SphereMask(
   A=TextureCoordinate()
   B=Constant() {r=0.5}
   Radius=ScalarParameter() {defaultValue=0.5, parameterName=SphereMask-Radius}
   Hardness=ScalarParameter() {defaultValue=0.25, parameterName=SphereMask-Hardness}
  )
  n2 = ScalarParameter() {defaultValue=1, parameterName=Color-SphereMask_Exp}
  n3 = Clamp(None=Multiply(A=Power(Base=@n1, Exp=@n2), B=ScalarParameter() {defaultValue=1, parameterName=Color-SphereMask_Mult}))
  n4 = WorldPosition()
  n5 = ActorPositionWS()
  n6 = Divide(
   A=ComponentMask(None=Subtract(A=@n4.XYZ, B=@n5)) {b=true}
   B=ScalarParameter() {defaultValue=5000, desc="This Value Should be Double the Decal Actors Z Size (keep scale 1)", parameterName=Grad_World-Z_Scale}
  ) {constB=500}
  n7 = ConstantBiasScale(None=Cosine(None=Frac(None=@n6)))
  Multiply(
   A=LinearInterpolate(
    A=VectorParameter() {parameterName=BaseColor_Dark, defaultValue=<rgba 0 0.25 1.5 1>}.RGB
    B=VectorParameter() {parameterName=BaseColor_Bright, defaultValue=<rgba 0 1.3042 4 1>}.RGB
    Alpha=Multiply(A=@n3, B=@n7)
   )
   B=ScalarParameter() {defaultValue=3, parameterName=ColorIntensity-Master}
  )
  n8 = ScalarParameter() {defaultValue=1, parameterName=Opacity-SphereMask_Exp}
  n9 = Clamp(None=Multiply(A=Power(Base=@n1, Exp=@n8), B=ScalarParameter() {defaultValue=1.5, parameterName=Opacity-SphereMask_Mult}))
  n10 = ScalarParameter() {defaultValue=1, parameterName=Opacity-Master}
  n11 = Constant() {r=1}
  n12 = Constant() {r=1}
  n13 = TextureSampleParameter2D(UVs=Multiply(A=TextureCoordinate(), B=AppendVector(A=@n11, B=@n12))) {group=Erosion, parameterName=Erosion Noise, samplerType=SAMPLERTYPE_LinearColor, sortPriority=3, texture=/FortWater/Textures/T_Caustic_M}
  n14 = Subtract(A=@n13.B, B=ScalarParameter() {defaultValue=-1, parameterName=Erosion_-1to1})
  Reroute(None=Multiply(A=Multiply(A=@n9, B=@n10), B=Saturate(None=@n14)))
 ]
}
